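# Control Rig for the SpecterDunk backpack cosmetic that adjusts the pose of its dangling skull strap based on a skydive value.
# The SkydiveFloat_Float variable passes through an AlphaInterp node, which smooths it at equal rates rising and falling.
# The smoothed result weights a single Modify Transforms unit that applies small rotation offsets to dyn_strap bones in the chain ending at the skull.
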ControlRigBlueprint /BRCosmetics/Accessories/FORT_Backpacks/Backpack_M_MED_SpecterDunk/Meshes/M_MED_SpecterDunk_Pack_CtrlRig
graphs [
 RigVMFunctionLibrary {links [], nodes []}
 RigVMModel {nodes [
   RigUnit_BeginExecution = &RigUnit_BeginExecution()
   AlphaInterp = &RigVMFunction_AlphaInterp(InterpSpeedDecreasing=5, InterpSpeedIncreasing=5, Scale=1, bClampResult=False, Value=RigVMVariableNode() {variable=SkydiveFloat_Float}.Value)
   &RigUnit_ModifyTransforms(
    ItemToModify="((Item=(Type=Bone,Name=\"dyn_strap_02\"),Transform=(Rotation=(X=0,Y=0,Z=0.0436,W=0.999),Translation=(X=0,Y=0,Z=0),Scale3D=(X=1,Y=1,Z=1))),(Item=(Type=Bone,Name=\"dyn_strap_03\"),Transform=(Rotation=(X=0,Y=0,Z=0.0436,W=0.999),Translation=(X=0,Y=0,Z=0),Scale3D=( ...(truncated)"
    WeightMinimum=0
    Weight=@AlphaInterp.Result
    ExecutePin=@RigUnit_BeginExecution.ExecutePin
   )
  ]}
]
hierarchy [
 root {loc=<0 0 0>, rot=<0 0 0 1>}
 pack {parent=root, loc=<0 -0.7316 0>, rot=<0 0 0 1>}
 base_pack {parent=pack, loc=<0 0 0>, rot=<0 0 0 1>}
 base_strap_01 {parent=base_pack, loc=<-12.3596 -0.8042 1.1101>, rot=<0.7025 -0.0892 -0.7005 -0.0884>}
 dyn_strap_01 {parent=base_strap_01, loc=<0 0 0>, rot=<0 0 0 1>}
 dyn_strap_02 {parent=dyn_strap_01, loc=<4.0512 0 0>, rot=<0 0 0 1>}
 dyn_strap_03 {parent=dyn_strap_02, loc=<3.6034 0 0>, rot=<0 0 0 1>}
 dyn_strap_04 {parent=dyn_strap_03, loc=<3.5548 0 0>, rot=<0 0 0 1>}
 dyn_strap_05 {parent=dyn_strap_04, loc=<3.5225 0 0>, rot=<0 0 0 1>}
 dyn_skull_01 {parent=dyn_strap_05, loc=<2.1023 0.0456 0.0162>, rot=<-0.0218 -0.0096 0.0001 -0.9997>}
]
variables [SkydiveFloat_Float {type=double, default=0}]
